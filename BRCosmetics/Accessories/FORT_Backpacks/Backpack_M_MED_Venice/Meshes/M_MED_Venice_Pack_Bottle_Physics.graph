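# Physics asset for the bottle on the Venice backpack cosmetic, named for letting the bottle move on the pack.
# A small capsule body on the dyn_bottle01 bone is joined by one constraint to a long, tilted capsule on the pack bone.
# The constraint carries narrow swing and twist limits, so the bottle has only a slight range of wobble relative to the pack.
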PhysicsAsset /BRCosmetics/Accessories/FORT_Backpacks/Backpack_M_MED_Venice/Meshes/M_MED_Venice_Pack_Bottle_Physics
bodies {
 dyn_bottle01 [{shapeType=Capsule, length=2, radius=2, shapeName=dyn_bottle01_capsule, center=<5.433 0 0>}]
 pack [{shapeType=Capsule, length=55.316, radius=5.0751, center=<2.4113 0 -1.6884>, rotation=<rot -35 0 -0.0001>}]
}
constraints [{bone1Name=dyn_bottle01, bone2Name=pack, swing1LimitDegrees=5, swing2LimitDegrees=3, twistLimitDegrees=10}]
