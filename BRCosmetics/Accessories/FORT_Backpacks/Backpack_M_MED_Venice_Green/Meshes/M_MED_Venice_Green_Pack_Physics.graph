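# Physics asset for the M_MED_Venice_Green backpack cosmetic, defining bodies and joints for its dangling feather and string bones.
# A long capsule on the pack bone anchors two master bodies, which root a six-segment feather chain and a nine-segment string chain of small capsules.
# Each segment is constrained to the previous one with moderate swing and very tight twist, except the last two string joints, which have much narrower swing and slightly looser twist.
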PhysicsAsset /BRCosmetics/Accessories/FORT_Backpacks/Backpack_M_MED_Venice_Green/Meshes/M_MED_Venice_Green_Pack_Physics
bodies {
 dyn_feather_01 [{shapeType=Capsule, length=0.8, radius=0.8, shapeName=dyn_feather_01_capsule, center=<1 0 0>, rotation=<rot 90 0 0>}]
 dyn_feather_02 [{shapeType=Capsule, length=0.8, radius=0.8, shapeName=dyn_feather_02_capsule, center=<1 0 0>, rotation=<rot 90 0 0>}]
 dyn_feather_03 [{shapeType=Capsule, length=0.8, radius=0.8, shapeName=dyn_feather_03_capsule, center=<1 0 0>, rotation=<rot 90 0 0>}]
 dyn_feather_04 [{shapeType=Capsule, length=0.8, radius=0.8, shapeName=dyn_feather_04_capsule, center=<1 0 0>, rotation=<rot 90 0 0>}]
 dyn_feather_05 [{shapeType=Capsule, length=0.8, radius=0.8, shapeName=dyn_feather_05_capsule, center=<1 0 0>, rotation=<rot 90 0 0>}]
 dyn_feather_06 [{shapeType=Capsule, length=0.8, radius=0.8, shapeName=dyn_feather_06_capsule, center=<1 0 0>, rotation=<rot 90 0 0>}]
 dyn_string_01 [{shapeType=Capsule, length=1.5, radius=0.8, shapeName=dyn_string_01_capsule, center=<2 0 0>, rotation=<rot 90 0 0>}]
 dyn_string_02 [{shapeType=Capsule, length=1.5, radius=0.8, shapeName=dyn_string_02_capsule, center=<2 0 0>, rotation=<rot 90 0 0>}]
 dyn_string_03 [{shapeType=Capsule, length=1.5, radius=0.8, shapeName=dyn_string_03_capsule, center=<2 0 0>, rotation=<rot 90 0 0>}]
 dyn_string_04 [{shapeType=Capsule, length=1.5, radius=0.8, shapeName=dyn_string_04_capsule, center=<2 0 0>, rotation=<rot 90 0 0>}]
 dyn_string_05 [{shapeType=Capsule, length=1.5, radius=0.8, shapeName=dyn_string_05_capsule, center=<2 0 0>, rotation=<rot 90 0 0>}]
 dyn_string_06 [{shapeType=Capsule, length=1.5, radius=0.8, shapeName=dyn_string_06_capsule, center=<2 0 0>, rotation=<rot 90 0 0>}]
 dyn_string_07 [{shapeType=Capsule, length=1.5, radius=0.8, shapeName=dyn_string_07_capsule, center=<2 0 0>, rotation=<rot 90 0 0>}]
 dyn_string_08 [{shapeType=Capsule, length=1.5, radius=1, shapeName=dyn_string_08_capsule, center=<0.5 0 0>, rotation=<rot 90 0 0>}]
 dyn_string_09 [{shapeType=Capsule, length=1.5, radius=1, shapeName=dyn_string_09_capsule, center=<2 0 0>, rotation=<rot 90 0 0>}]
 master_feather [{shapeType=Capsule, length=0.505, radius=0.505, shapeName=master_feather_capsule, rotation=<rot 0 -90 0>}]
 master_stringA [{shapeType=Capsule, length=0.505, radius=0.505, shapeName=master_stringA_capsule, rotation=<rot 0 -90 0>}]
 pack [{shapeType=Capsule, length=55.316, radius=3, center=<7.6437 -13.8439 3.5204>, rotation=<rot -32.4123 0.0002 -0.0003>}]
}
constraints [
 {bone1Name=dyn_feather_01, bone2Name=master_feather, swing1LimitDegrees=50, swing2LimitDegrees=50, twistLimitDegrees=1}
 {bone1Name=dyn_feather_02, bone2Name=dyn_feather_01, swing1LimitDegrees=45, swing2LimitDegrees=45, twistLimitDegrees=1}
 {bone1Name=dyn_feather_03, bone2Name=dyn_feather_02, swing1LimitDegrees=45, swing2LimitDegrees=45, twistLimitDegrees=1}
 {bone1Name=dyn_feather_04, bone2Name=dyn_feather_03, swing1LimitDegrees=45, swing2LimitDegrees=45, twistLimitDegrees=1}
 {bone1Name=dyn_feather_05, bone2Name=dyn_feather_04, swing1LimitDegrees=45, swing2LimitDegrees=45, twistLimitDegrees=1}
 {bone1Name=dyn_feather_06, bone2Name=dyn_feather_05, swing1LimitDegrees=45, swing2LimitDegrees=45, twistLimitDegrees=1}
 {bone1Name=dyn_string_01, bone2Name=master_stringA, swing1LimitDegrees=40, swing2LimitDegrees=40, twistLimitDegrees=1}
 {bone1Name=dyn_string_02, bone2Name=dyn_string_01, swing1LimitDegrees=45, swing2LimitDegrees=45, twistLimitDegrees=1}
 {bone1Name=dyn_string_03, bone2Name=dyn_string_02, swing1LimitDegrees=45, swing2LimitDegrees=45, twistLimitDegrees=1}
 {bone1Name=dyn_string_04, bone2Name=dyn_string_03, swing1LimitDegrees=45, swing2LimitDegrees=45, twistLimitDegrees=1}
 {bone1Name=dyn_string_05, bone2Name=dyn_string_04, swing1LimitDegrees=45, swing2LimitDegrees=45, twistLimitDegrees=1}
 {bone1Name=dyn_string_06, bone2Name=dyn_string_05, swing1LimitDegrees=45, swing2LimitDegrees=45, twistLimitDegrees=1}
 {bone1Name=dyn_string_07, bone2Name=dyn_string_06, swing1LimitDegrees=45, swing2LimitDegrees=45, twistLimitDegrees=1}
 {bone1Name=master_feather, bone2Name=pack, swing1LimitDegrees=45, swing2LimitDegrees=45, twistLimitDegrees=45}
 {bone1Name=master_stringA, bone2Name=pack, swing1LimitDegrees=45, swing2LimitDegrees=45, twistLimitDegrees=45}
 {bone1Name=dyn_string_08, bone2Name=dyn_string_07, swing1LimitDegrees=10, swing2LimitDegrees=10, twistLimitDegrees=5}
 {bone1Name=dyn_string_09, bone2Name=dyn_string_08, swing1LimitDegrees=10, swing2LimitDegrees=10, twistLimitDegrees=5}
]
