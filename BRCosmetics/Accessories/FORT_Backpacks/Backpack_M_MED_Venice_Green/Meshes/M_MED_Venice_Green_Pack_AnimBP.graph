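# Animation Blueprint for the Venice Green backpack cosmetic, driving pack offsets, spinning wheels and secondary motion.
# RigidBody nodes simulate pack and bottle with overridden gravity, weighted by skydiving state; fixed string and feather rotations apply when rigid bodies are disabled.
# Wheel spin accumulates per frame, its rate eased up during jumps or melee.
# Weapon, customization and skydive events toggle material parameter Hide Element 01, checking for a Venice-tagged pickaxe.
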AnimBlueprint /BRCosmetics/Accessories/FORT_Backpacks/Backpack_M_MED_Venice_Green/Meshes/M_MED_Venice_Green_Pack_AnimBP
target_skeleton=&/M_MED_Venice_Green_Pack_Skeleton
graphs [
 AnimGraph {nodes [
   n1 = "Transform (Modify) Bone - Bone: pack"(ComponentPose=Local To Component(LocalPose=Input Pose()), Translation="0.000,6.146,-0.000", Rotation="0,0,0") {translationMode=BMM_Additive, alphaBoolBlend {blendOption=Linear}, boneToModify {boneName=pack}, translation {y=6.146}}
   n2 = Get Is Sky Diving() -> (Is Sky Diving:double)
   n3 = "Transform (Modify) Bone - Bone: pack"(ComponentPose=@n1, Alpha=@n2, Translation="0.000,-4.336,-0.000", Rotation="0,0.000,15") {
    rotationMode=BMM_Additive
    translationMode=BMM_Additive
    alphaBoolBlend {blendOption=Linear}
    boneToModify {boneName=pack}
    rotation {roll=15}
    translation {y=-4.336}
   }
   n4 = Get wheel_rotate() -> (wheel_rotate:Rotator)
   n5 = "Transform (Modify) Bone - Bone: wheel_r_Dyn"(ComponentPose=@n3, Alpha=1, Translation="0,0,0", Rotation=@n4, Scale="1,1,1") {
    rotationMode=BMM_Additive
    rotationSpace=BCS_BoneSpace
    translationMode=BMM_Additive
    translationSpace=BCS_BoneSpace
    alphaBoolBlend {blendOption=Linear}
    boneToModify {boneName=wheel_r_Dyn}
   }
   n6 = "Transform (Modify) Bone - Bone: wheel_l_Dyn"(ComponentPose=@n5, Alpha=1, Translation="0,0,0", Rotation=@n4, Scale="1,1,1") {
    rotationMode=BMM_Additive
    rotationSpace=BCS_BoneSpace
    translationMode=BMM_Additive
    translationSpace=BCS_BoneSpace
    alphaBoolBlend {blendOption=Linear}
    boneToModify {boneName=wheel_l_Dyn}
   }
   n7 = Get GravityOverride()
   n8 = Get Negate Is Sky Diving() -> (Negate Is Sky Diving:double)
   n9 = RigidBody!(ComponentPose=@n6, Alpha=@n8, OverrideWorldGravity:Vector=@n7, ExternalForce) -> (Pose) {
    bClampLinearTranslationLimitToRefPose=true
    bOverrideWorldGravity=true
    overridePhysicsAsset=&/M_MED_Venice_Green_Pack_Physics
    simulationSpace=BaseBoneSpace
    alphaBoolBlend {blendOption=Linear}
    alphaScaleBiasClamp {bInterpResult=true}
    baseBoneRef {boneName=root}
    componentAppliedLinearAccClamp=<3000 3000 3000>
    componentLinearAccScale=<0.4 0.3 0.3>
    componentLinearVelScale=<0.4 0.3 0.3>
    overrideWorldGravity {z=-980}
    simSpaceSettings {
     maxAngularAcceleration=0
     maxAngularVelocity=0.3
     maxLinearAcceleration=100
     maxLinearVelocity=100
     velocityScaleZ=0
     worldAlpha=1
     externalLinearDragV=<0.2 0.2>
    }
   }
   n10 = RigidBody(ComponentPose=@n9, Alpha=@n2, OverrideWorldGravity=@n7) {
    bClampLinearTranslationLimitToRefPose=true
    bOverrideWorldGravity=true
    overridePhysicsAsset=&M_MED_Venice_Green_Pack_Physics
    simulationSpace=BaseBoneSpace
    alphaBoolBlend {blendOption=Linear}
    alphaScaleBiasClamp {bInterpResult=true}
    baseBoneRef {boneName=root}
    componentAppliedLinearAccClamp=<3000 3000 3000>
    componentLinearAccScale=<0.4 0.3 0.3>
    componentLinearVelScale=<0.4 0.3 0.3>
    overrideWorldGravity {z=-980}
    simSpaceSettings {
     maxAngularAcceleration=0
     maxAngularVelocity=0.3
     maxLinearAcceleration=100
     maxLinearVelocity=100
     velocityScaleZ=0
     worldAlpha=1
     externalLinearDragV=<0.2 0.2>
    }
   }
   n11 = RigidBody(ComponentPose=@n10, Alpha=@n8, OverrideWorldGravity=@n7) {
    bClampLinearTranslationLimitToRefPose=true
    bOverrideWorldGravity=true
    overridePhysicsAsset=/BRCosmetics/Accessories/FORT_Backpacks/Backpack_M_MED_Venice/Meshes/M_MED_Venice_Pack_Bottle_Physics
    simulationSpace=BaseBoneSpace
    alphaBoolBlend {blendOption=Linear}
    baseBoneRef {boneName=root}
    componentAppliedLinearAccClamp=<3000 3000 3000>
    componentLinearAccScale=<0.4 0.3 0.3>
    componentLinearVelScale=<0.4 0.3 0.3>
    overrideWorldGravity {z=-980}
    simSpaceSettings {
     maxAngularAcceleration=0
     maxAngularVelocity=0.3
     maxLinearAcceleration=100
     maxLinearVelocity=100
     velocityScaleZ=0
     worldAlpha=1
     externalLinearDragV=<0.2 0.2>
    }
   }
   n12 = "Transform (Modify) Bone - Bone: root"(ComponentPose=@n11, Translation="0,-2,0", Rotation="0,0,0") {
    translationMode=BMM_Additive
    translationSpace=BCS_BoneSpace
    alphaBoolBlend {blendOption=Linear}
    boneToModify {boneName=root}
    translation {y=-2}
   }
   n13 = "Transform (Modify) Bone - Bone: root"(ComponentPose=@n12, Translation="0,-2,0", Rotation="0,0,0") {
    translationMode=BMM_Additive
    translationSpace=BCS_BoneSpace
    alphaBoolBlend {blendOption=Linear}
    boneToModify {boneName=root}
    translation {y=-2}
   }
   n14 = NOT Boolean(A=Get bRigidBodyEnabled())
   n15 = "Transform (Modify) Bone - Bone: dyn_string_01"!(ComponentPose=@n13, bAlphaBoolEnabled:bool=@n14, Translation="0,0,0", Rotation="-48.173,-116.474,41.881", Scale="1,1,1") -> (Pose) {
    alphaInputType=Bool
    rotationMode=BMM_Additive
    translationSpace=BCS_BoneSpace
    alphaBoolBlend {blendOption=Linear}
    boneToModify {boneName=dyn_string_01}
    rotation=<rot -48.173 -116.474 41.881>
   }
   n16 = "Transform (Modify) Bone - Bone: dyn_string_02"!(ComponentPose=@n15, bAlphaBoolEnabled:bool=@n14, Translation="0,0,0", Rotation="-20.759,27.903,-67.203", Scale="1,1,1") -> (Pose) {
    alphaInputType=Bool
    rotationMode=BMM_Additive
    translationSpace=BCS_BoneSpace
    alphaBoolBlend {blendOption=Linear}
    boneToModify {boneName=dyn_string_02}
    rotation=<rot -20.759 27.903 -67.203>
   }
   n17 = "Transform (Modify) Bone - Bone: dyn_string_03"!(ComponentPose=@n16, bAlphaBoolEnabled:bool=@n14, Translation="0,0,0", Rotation="-5.158,-0.000,-30.000", Scale="1,1,1") -> (Pose) {
    alphaInputType=Bool
    rotationMode=BMM_Additive
    translationSpace=BCS_BoneSpace
    alphaBoolBlend {blendOption=Linear}
    boneToModify {boneName=dyn_string_03}
    rotation {pitch=-5.158, roll=-30}
   }
   n18 = "Transform (Modify) Bone - Bone: dyn_feather_01"!(ComponentPose=@n17, bAlphaBoolEnabled:bool=@n14, Translation="0,0,0", Rotation="20.000,8.689,-170.000", Scale="1,1,1") -> (Pose) {
    alphaInputType=Bool
    rotationMode=BMM_Additive
    alphaBoolBlend {blendOption=Linear}
    boneToModify {boneName=dyn_feather_01}
    rotation=<rot 20 8.689 -170>
   }
   n19 = "Transform (Modify) Bone - Bone: dyn_feather_02"!(ComponentPose=@n18, bAlphaBoolEnabled:bool=@n14, Translation="0,0,0", Rotation="-54.333,10.151,-12.220", Scale="1,1,1") -> (Pose) {
    alphaInputType=Bool
    rotationMode=BMM_Additive
    translationSpace=BCS_BoneSpace
    alphaBoolBlend {blendOption=Linear}
    boneToModify {boneName=dyn_feather_02}
    rotation=<rot -54.333 10.151 -12.22>
   }
   n20 = "Transform (Modify) Bone - Bone: dyn_feather_03"!(ComponentPose=@n19, bAlphaBoolEnabled:bool=@n14, Translation="0,0,0", Rotation="-10.242,0.000,-0.000", Scale="1,1,1") -> (Pose) {
    alphaInputType=Bool
    rotationMode=BMM_Additive
    translationSpace=BCS_BoneSpace
    alphaBoolBlend {blendOption=Linear}
    boneToModify {boneName=dyn_feather_03}
    rotation {pitch=-10.242}
   }
   Output Pose(Result=Component To Local(ComponentPose=@n20)) {layerGroup=None}
  ]}
 EventGraph {nodes [
   n1 = Event BlueprintUpdateAnimation()
   > n2 = Sequence!(execute) -> (then_0, then_1, then_2:Exec)
   > Set Delta Time X(Delta Time X=@n1.DeltaTimeX)
   > n3 = Gravity Override State Function DEPRECATED(CharacterPartAnimInstance=Self-Reference(), GameTimeInSeconds=GetGameTimeInSeconds(), GravityOverrideParams=Get Gravity Override Params())
   > Set GravityOverride(GravityOverride=@n3)
   n4 = Get bIsSkyDiving()
   > Set Is Sky Diving(Is Sky Diving:double="To Float (Boolean)"(InBool=@n4)) -> (Output_Get:double)
   > Set Negate Is Sky Diving(Negate Is Sky Diving:double="To Float (Boolean)"(InBool=NOT Boolean(A=@n4))) -> (Output_Get:double)
   n5 = OR Boolean!(A=Get bIsJumping(), B=Get bIsPlayingMeleeAnim(), C:bool, D:bool) -> (ReturnValue)
   n6 = Set Is Playing Melee Anim(
    execute=@n2.then_2
    Is Playing Melee Anim:double=FInterpTo(
     Current=Get Is Playing Melee Anim() -> (Is Playing Melee Anim:double)
     Target="To Float (Boolean)"(InBool=@n5)
     DeltaTime=@n1.DeltaTimeX
     InterpSpeed=1.5
    )
   ) -> (Output_Get:double)
   n7 = float * float(A=1, B=Lerp(A=2, B=1200, Alpha=@n6))
   n8 = Branch(execute=@n2.then_1, Condition="float >= float"(A=Get wheel_spin() -> (wheel_spin:double), B=@n7))
   n9 = float + float(A=Get wheel_spin(), B=float * float(A=@n7, B=@n1.DeltaTimeX) {is=title K2Node_PromotableOperator}) {is=title K2Node_CommutativeAssociativeBinaryOperator}
   > n10 = Set wheel_spin(wheel_spin:double=@n9) -> (Output_Get:double)
   Set wheel_rotate(execute=[@n11.then, @n10.then], wheel_rotate:Rotator=MakeRotator(Roll=@n9)) -> (Output_Get:Rotator)
   n11 = Set wheel_spin(execute=@n8.else, wheel_spin=@n9)
   n12 = Event OnWeaponChanged()
   n13 = Get Dynamic Material Instance(execute=[@n12.then, @n21.then], SkeletalMeshComponent=GetPartSkeletalMeshComponent())
   > n14 = Sequence!(execute) -> (then_0, then_1)
   n15 = PickaxeItemDefinitionHasAnyMetaTag(Tags=Make Array("[0]"=(TagName="Cosmetics.MetaTags.Venice")), bRequiresEquipped=true)
   > n16 = Branch(Condition=@n15)
   > n17 = Branch(Condition=Get bIsSkyDiving())
   SetScalarParameterValue(execute=[@n19.then, @n17.then, @n18.else, @n16.else], self=@n13, ParameterName=Hide Element 01, Value=0)
   SetScalarParameterValue(execute=[@n17.else, @n18.then], self=@n13, ParameterName=Hide Element 01, Value=1)
   n18 = Branch(execute=[@n20.then, @n14.then_1], Condition=@n15)
   n19 = Event OnBeginSkydiving()
   n20 = Event OnEndSkydiving()
   n21 = Event OnCharacterCustomizationLoaded()
   TryGetPawnOwner()
  ]}
]
variables [
 Delta Time X {type=number}
 Gravity Override Params {type=GravityOverrideParamsStruct}
 GravityOverride {type=Vector}
 Is Playing Melee Anim {type=number}
 Is Sky Diving {type=number}
 Negate Is Sky Diving {type=number}
 wheel_rotate {type=Rotator}
 wheel_spin {type=number}
]
